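# Control Rig for the OmegaAngel female body cosmetic that adds procedural corrections to skirt, front cloth, crown, bracelet, leg and feather bones.
# It runs a shared DeformRig function, then derives crouch weights from pelvis height and Kalman-smoothed directional movement weights.
# Weighted transform offsets follow these values, joint angles, bone distances and thigh spherical pose readers.
# State flags such as skydive, parachute, DBNO, emote and IsALT gate the corrections.
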ControlRigBlueprint /BRCosmetics/Characters/Player/Female/Medium/Bodies/F_MED_OmegaAngel/Meshes/F_MED_OmegaAngel_CtrlRig
graphs [
 Multiply_5_ContainedGraph {nodes [
   n1 = RigVMFunctionEntryNode()
   Multiply = &RigVMFunction_MathDoubleMul(A=@n1.A, B=@n1.B)
   RigVMFunctionReturnNode(Result=&RigVMFunction_MathDoubleMul(B=@n1.C, A=@Multiply.Result).Result)
  ]}
 Multiply_5_ContainedGraph {nodes [
   n1 = RigVMFunctionEntryNode()
   Multiply = &RigVMFunction_MathDoubleMul(A=@n1.A, B=@n1.B)
   RigVMFunctionReturnNode(Result=&RigVMFunction_MathDoubleMul(B=@n1.C, A=@Multiply.Result).Result)
  ]}
 Multiply_5_ContainedGraph {nodes [
   n1 = RigVMFunctionEntryNode()
   Multiply = &RigVMFunction_MathDoubleMul(A=@n1.A, B=@n1.B)
   RigVMFunctionReturnNode(Result=&RigVMFunction_MathDoubleMul(B=@n1.C, A=@Multiply.Result).Result)
  ]}
 RigVMFunctionLibrary {links [], nodes []}
 RigVMFunction_Sequence_1_ContainedGraph {nodes [
   n1 = RigVMFunctionEntryNode()
   RigVMFunction_Sequence = &RigVMFunction_Sequence(ExecuteContext=@n1.ExecuteContext)
   n2 = &RigVMFunction_Sequence(ExecuteContext=@RigVMFunction_Sequence.B)
   n3 = &RigVMFunction_Sequence(ExecuteContext=@n2.B)
   n4 = &RigVMFunction_Sequence(ExecuteContext=@n3.B)
   n5 = &RigVMFunction_Sequence(ExecuteContext=@n4.B)
   n6 = &RigVMFunction_Sequence(ExecuteContext=@n5.B)
   RigVMFunctionReturnNode(A=@RigVMFunction_Sequence.A, B=@n2.A, C=@n3.A, D=@n4.A, E=@n5.A, F=@n6.A, G=@n6.B)
  ]}
 RigVMFunction_Sequence_3_ContainedGraph {nodes [
   n1 = RigVMFunctionEntryNode()
   n2 = &RigVMFunction_Sequence(ExecuteContext=@n1.ExecuteContext)
   n3 = &RigVMFunction_Sequence(ExecuteContext=@n2.B)
   n4 = &RigVMFunction_Sequence(ExecuteContext=@n3.B)
   n5 = &RigVMFunction_Sequence(ExecuteContext=@n4.B)
   RigVMFunctionReturnNode(B=@n3.A, A=@n2.A, C=@n4.A, D=@n5.A, E=@n5.B)
  ]}
 RigVMFunction_Sequence_5_ContainedGraph {nodes [
   n1 = RigVMFunctionEntryNode()
   n2 = &RigVMFunction_Sequence(ExecuteContext=@n1.ExecuteContext)
   n3 = &RigVMFunction_Sequence(ExecuteContext=@n2.B)
   RigVMFunctionReturnNode(B=@n3.A, C=@n3.B, A=@n2.A)
  ]}
 RigVMModel {nodes [
   RigUnit_BeginExecution = &RigUnit_BeginExecution()
   FN_DeformRig_FNC = RigVMFunctionReferenceNode(
    DeformRig_ Curve="(deformRig_l_upperarm_up_alpha_72_#1=False,deformRig_r_upperarm_up_alpha_73_#2=False,deformRig_l_upperarm_fr_alpha_74_#3=False,deformRig_r_upperarm_fr_alpha_75_CB51B8 ...(truncated)"
    DeformRig_Bool="(upperarm_1_#4=True,elbow_3_#5=True,wrist_7_#6=True,groin_8_#7=True,glute_17_#8=True,knee_9_40C408324CA261 ...(truncated)"
    DeformRig_Debug=(upperarm_debug_18_#4=False,wrist_debug_20_#6=False,lat_debug_26_#9=False,groinGlute_debug_27_#7=False)
    DeformRig_Value="(upperarm_up_18_#10=(Rotation=(X=0,Y=0,Z=0,W=1),Translation=(X=0,Y=0,Z=8),Scale3D=(X=0.7,Y=1,Z=1)),upperarm_fwd_fr_19_#11=(Rotation=(X=0,Y=0,Z=0,W=1),Translation=(X=0,Y=10,Z=0),Scale3D=(X=0.5,Y=1,Z= ...(truncated)"
    ExecuteContext=@RigUnit_BeginExecution.ExecutePin
   )
   n1 = &RigUnit_GetRelativeTransformForItem(Child=(Type=Bone,Name="pelvis"), Parent=(Type=Bone,Name="root"), bChildInitial=False, bParentInitial=False)
   > RigVMVariableNode(Value=&RigVMFunction_MathDoubleMul(
     B=&RigVMFunction_MathDoubleRemap(SourceMaximum=67, SourceMinimum=90, TargetMaximum=1, TargetMinimum=0, Value=@n1.RelativeTransform.Translation.Z).Result
     A=&RigVMFunction_MathDoubleRemap(SourceMaximum=69, SourceMinimum=50, TargetMaximum=1, TargetMinimum=0, Value=@n1.RelativeTransform.Translation.Z).Result
    ).Result) {variable=CrouchWalk}
   > RigVMVariableNode(Value=&RigVMFunction_MathDoubleRemap(
     SourceMaximum=39
     SourceMinimum=70
     TargetMaximum=1
     TargetMinimum=0
     Value=&RigUnit_GetRelativeTransformForItem(Child=(Type=Bone,Name="pelvis"), Parent=(Type=Bone,Name="root"), bChildInitial=False, bParentInitial=False).RelativeTransform.Translation.Z
    ).Result) {variable=Crouch}
   RigVMFunction_MathBoolToFloat = &RigVMFunction_MathBoolToFloat(Value=RigVMVariableNode() {variable=FE}.Value)
   > &RigUnit_ModifyTransforms(
    ItemToModify="((Item=(Type=Bone,Name=\"driver_skirt_rear_01_l\"),Transform=(Rotation=(X=0,Y=0,Z=0,W=1),Translation=(X=0,Y=-2,Z=0),Scale3D=(X=1,Y=1,Z=1))),(Item=(Type=Bone,Name=\"driver_skirt_rear_01_r\"),Transform=(Rotation=(X=0,Y=0,Z=0,W=1),Translation=(X=0,Y=-2,Z=0),Scale ...(truncated)"
    WeightMinimum=0
    Weight=@RigVMFunction_MathBoolToFloat.Result
   )
   > RigVMFunction_Sequence_4 = RigVMAggregateNode()
   RigVMFunction_Sequence_2 = RigVMAggregateNode(ExecuteContext=@RigVMFunction_Sequence_4.A)
   n2 = RigVMVariableNode() {variable=ForwardBackwardMove}
   Remap = &RigVMFunction_MathDoubleRemap(SourceMaximum=1, SourceMinimum=0, TargetMaximum=1, TargetMinimum=0, Value=@n2.Value)
   KalmanFilter = &RigVMFunction_KalmanFloat(BufferSize=8, Value=@Remap.Result) {template="KalmanFilter::Execute(in Value,in BufferSize,out Result)"}
   RigVMVariableNode(Value=@KalmanFilter.Result, ExecuteContext=@RigVMFunction_Sequence_2.A) {variable=ForwardMovement}
   > RigVMVariableNode(Value=&RigVMFunction_KalmanFloat(
     BufferSize=8
     Value=&RigVMFunction_MathDoubleRemap(SourceMaximum=0, SourceMinimum=-1, TargetMaximum=0, TargetMinimum=1, Value=@n2.Value).Result
    ) {template="KalmanFilter::Execute(in Value,in BufferSize,out Result)"}.Result) {variable=BackwardMovement}
   n3 = RigVMVariableNode() {variable=LeftRightMove}
   > RigVMVariableNode(Value=&RigVMFunction_KalmanFloat(
     BufferSize=8
     Value=&RigVMFunction_MathDoubleRemap(SourceMaximum=0, SourceMinimum=-1, TargetMaximum=0, TargetMinimum=1, Value=@n3.Value).Result
    ) {template="KalmanFilter::Execute(in Value,in BufferSize,out Result)"}.Result) {variable=LeftMovement}
   > RigVMVariableNode(Value=&RigVMFunction_KalmanFloat(
     BufferSize=8
     Value=&RigVMFunction_MathDoubleRemap(SourceMaximum=1, SourceMinimum=0, TargetMaximum=1, TargetMinimum=0, Value=@n3.Value).Result
    ) {template="KalmanFilter::Execute(in Value,in BufferSize,out Result)"}.Result) {variable=RightMovement}
   > RigVMFunction_Sequence = RigVMAggregateNode()
   n4 = &RigUnit_ModifyTransforms(
    ItemToModify="((Item=(Type=Bone,Name=\"Base_skirt_frontMid_l\"),Transform=(Rotation=(X=0,Y=0,Z=0,W=1),Translation=(X=0,Y=0,Z=-1.5),Scale3D=(X=1,Y=1,Z=1))),(Item=(Type=Bone,Name=\"Base_skirt_frontMid_r\"),Transform=(Rotation=(X=0,Y=0,Z=0,W=1),Translation=(X=0,Y=0,Z=1.5),Scal ...(truncated)"
    Weight=1
    WeightMinimum=0
    ExecutePin=@RigVMFunction_Sequence.A
   )
   n5 = &RigVMFunction_ControlFlowBranch(ExecuteContext=@n4.ExecutePin, Condition=RigVMVariableNode() {variable=IsALT}.Value)
   n6 = &RigVMFunction_ControlFlowBranch(ExecuteContext=@n5.False, Condition=RigVMVariableNode() {variable=Is parachute opened}.Value)
   Multiply = RigVMAggregateNode(
    A=&RigVMFunction_MathDoubleRemap(
     SourceMaximum=81
     SourceMinimum=80
     TargetMaximum=1
     TargetMinimum=0
     Value=&RigUnit_GetRelativeTransformForItem(Child=(Type=Bone,Name="pelvis"), Parent=(Type=Bone,Name="root"), bChildInitial=False, bParentInitial=False).RelativeTransform.Translation.Z
    ).Result
    C=&RigVMFunction_MathDoubleRemap(SourceMaximum=400, SourceMinimum=250, TargetMaximum=1, TargetMinimum=0, Value=RigVMVariableNode() {variable=Velocity character}.Value).Result
    B=RigVMVariableNode() {variable=ForwardMovement}.Value
   )
   n7 = &RigUnit_ModifyTransforms(
    ItemToModify="((Item=(Type=Bone,Name=\"dyn_driver_cloth_front\"),Transform=(Rotation=(X=0.0199,Y=-0.5732,Z=-0.0284,W=0.8187),Translation=(X=0,Y=0,Z=0),Scale3D=(X=1,Y=1,Z=1))),(Item=(Type=Bone,Name=\"driver_skirt_rear_01_l\"),Transform=(Rotation=(X=0,Y=0,Z=-0.5,W=0.866),Tran ...(truncated)"
    WeightMinimum=0
    Weight=&RigVMFunction_KalmanFloat(BufferSize=32, Value=@Multiply.Result) {template="KalmanFilter::Execute(in Value,in BufferSize,out Result)"}.Result
    ExecutePin=@n6.False
   )
   n8 = &RigVMFunction_ControlFlowBranch(ExecuteContext=@n7.ExecutePin, Condition=RigVMVariableNode() {variable=isBackPack}.Value)
   Get Delta Ref_2 = RigVMFunctionReferenceNode(Item=(Type=Bone,Name="pelvis"), ExecuteContext=@n8.True)
   > &RigUnit_ModifyTransforms(
    ItemToModify=((Item=(Type=Bone,Name="crown_l"),Transform=(Rotation=(X=0,Y=0,Z=0,W=1),Translation=(X=4,Y=0,Z=2),Scale3D=(X=1,Y=1,Z=1))),(Item=(Type=Bone,Name="crown_r"),Transform=(Rotation=(X=0,Y=0,Z=0,W=1),Translation=(X=-4,Y=0,Z=-2),Scale3D=(X=1,Y=1,Z=1))))
    WeightMinimum=0
    Weight=&RigVMFunction_KalmanFloat(
     BufferSize=32
     Value=&RigVMFunction_MathDoubleRemap(SourceMaximum=80, SourceMinimum=60, TargetMaximum=1, TargetMinimum=0, Value=@"Get Delta Ref_2".Result.Rotation.Yaw).Result
    ) {template="KalmanFilter::Execute(in Value,in BufferSize,out Result)"}.Result
   )
   n9 = &RigVMFunction_ControlFlowBranch(ExecuteContext=@RigVMFunction_Sequence.B, Condition=RigVMVariableNode() {variable=IsALT}.Value)
   n10 = &RigVMFunction_ControlFlowBranch(ExecuteContext=@n9.False, Condition=RigVMVariableNode() {variable=IsEmote}.Value)
   n11 = &RigVMFunction_ControlFlowBranch(ExecuteContext=@n10.False, Condition=RigVMVariableNode() {variable=IsDBNO}.Value)
   n12 = &RigVMFunction_ControlFlowBranch(ExecuteContext=@n11.False, Condition=RigVMVariableNode() {variable=Is parachute opened}.Value)
   n13 = RigVMVariableNode() {variable=Crouch}
   &RigUnit_ModifyTransforms(
    ItemToModify="((Item=(Type=Bone,Name=\"dyn_driver_cloth_front\"),Transform=(Rotation=(X=-0.0275,Y=-0.7303,Z=-0.1676,W=0.6616),Translation=(X=0.0482,Y=-0.3922,Z=1.0744),Scale3D=(X=1,Y=1,Z=1))),(Item=(Type=Bone,Name=\"Base_skirt_frontMid_r\"),Transform=(Rotation=(X=-0.2247,Y= ...(truncated)"
    WeightMinimum=0
    ExecutePin=@n12.False
    Weight=@n13.Value
   )
   > &RigUnit_ModifyTransforms(
    ItemToModify="((Item=(Type=Bone,Name=\"Driver_skirt_side_02_r1\"),Transform=(Rotation=(X=0.5712,Y=-0.2188,Z=0.2664,W=0.7449),Translation=(X=0.1506,Y=6.7734,Z=-5.3919),Scale3D=(X=1,Y=1,Z=1))),(Item=(Type=Bone,Name=\"Driver_skirt_sideMid_02_r\"),Transform=(Rotation=(X=0.0069, ...(truncated)"
    WeightMinimum=0
    Weight=@n13.Value
   )
   > &RigUnit_ModifyTransforms(
    ItemToModify=((Item=(Type=Bone,Name="dyn_driver_cloth_front"),Transform=(Rotation=(X=0,Y=-0.8578,Z=0,W=0.514),Translation=(X=-2.118,Y=3.1066,Z=3.9322),Scale3D=(X=1,Y=1,Z=1))))
    WeightMinimum=0
    Weight=RigVMVariableNode() {variable=CrouchWalk}.Value
   )
   n14 = &RigVMFunction_Sequence(ExecuteContext=@RigVMFunction_Sequence.C)
   n15 = &RigUnit_SphericalPoseReader(
    ActiveRegionScaleFactors=(PositiveWidth=3,NegativeWidth=1,PositiveHeight=3,NegativeHeight=3)
    ActiveRegionSize=0.05
    Debug=(bDrawDebug=false,DebugScale=25,DebugSegments=20,DebugThickness=0.25)
    DriverAxis=(X=-1,Y=0,Z=0)
    DriverItem=(Type=Bone,Name="thigh_l")
    FalloffRegionScaleFactors=(PositiveWidth=1,NegativeWidth=1,PositiveHeight=1,NegativeHeight=1)
    FalloffSize=0.4
    FlipHeightScaling=false
    OptionalParentItem=(Type=Bone)
    RotationOffset=(X=110,Y=-30,Z=0)
    ExecutePin=@n14.A
   )
   n16 = &RigVMFunction_MathDoubleRemap(
    SourceMaximum=81
    SourceMinimum=80
    TargetMaximum=1
    TargetMinimum=0
    Value=&RigUnit_GetRelativeTransformForItem(Child=(Type=Bone,Name="pelvis"), Parent=(Type=Bone,Name="root"), bChildInitial=False, bParentInitial=False).RelativeTransform.Translation.Z
   )
   > &RigUnit_ModifyTransforms(
    ItemToModify=((Item=(Type=Bone,Name="dyn_parent_cloth_front"),Transform=(Rotation=(X=-0.1736,Y=0,Z=0,W=0.9848),Translation=(X=0,Y=0,Z=0),Scale3D=(X=1,Y=1,Z=1))))
    WeightMinimum=0
    Weight=&RigVMFunction_MathDoubleMul(A=@n16.Result, B=@n15.OutputParam).Result
   )
   > n17 = &RigUnit_SphericalPoseReader(
    ActiveRegionScaleFactors=(PositiveWidth=3,NegativeWidth=1,PositiveHeight=3,NegativeHeight=3)
    ActiveRegionSize=0.05
    Debug=(bDrawDebug=false,DebugScale=25,DebugSegments=20,DebugThickness=0.25)
    DriverAxis=(X=-1,Y=0,Z=0)
    DriverItem=(Type=Bone,Name="thigh_l")
    FalloffRegionScaleFactors=(PositiveWidth=1,NegativeWidth=1,PositiveHeight=1,NegativeHeight=1)
    FalloffSize=0.4
    FlipHeightScaling=false
    OptionalParentItem=(Type=Bone)
    RotationOffset=(X=110,Y=-30,Z=0)
   )
   n18 = RigVMVariableNode() {variable=CrouchWalk}
   > &RigUnit_ModifyTransforms(
    ItemToModify=((Item=(Type=Bone,Name="dyn_parent_cloth_front"),Transform=(Rotation=(X=0,Y=0,Z=0,W=1),Translation=(X=0,Y=0,Z=0),Scale3D=(X=1,Y=1,Z=1))))
    WeightMinimum=0
    Weight=&RigVMFunction_MathDoubleMul(B=@n17.OutputParam, A=@n18.Value).Result
   )
   n19 = &RigUnit_SphericalPoseReader(
    ActiveRegionScaleFactors=(PositiveWidth=1,NegativeWidth=3,PositiveHeight=3,NegativeHeight=3)
    ActiveRegionSize=0.05
    Debug=(bDrawDebug=false,bDraw2D=false,DebugScale=25,DebugSegments=20,DebugThickness=0.25)
    DriverAxis=(X=1,Y=0,Z=0)
    DriverItem=(Type=Bone,Name="thigh_r")
    FalloffRegionScaleFactors=(PositiveWidth=1,NegativeWidth=1,PositiveHeight=1,NegativeHeight=1)
    FalloffSize=0.4
    FlipHeightScaling=false
    OptionalParentItem=(Type=Bone)
    RotationOffset=(X=-82.1368,Y=-4.5092,Z=12.8014)
    ExecutePin=@n14.B
   )
   > &RigUnit_ModifyTransforms(
    ItemToModify="((Item=(Type=Bone,Name=\"dyn_parent_cloth_front\"),Transform=(Rotation=(X=0,Y=-0.2588,Z=0,W=0.9659),Translation=(X=0,Y=0,Z=0),Scale3D=(X=1,Y=1,Z=1))),(Item=(Type=Bone,Name=\"dyn_skirt_sideMid_03_l1\"),Transform=(Rotation=(X=0,Y=0,Z=0,W=1),Translation=(X=0,Y=0, ...(truncated)"
    WeightMinimum=0
    Weight=&RigVMFunction_MathDoubleMul(A=@n16.Result, B=@n19.OutputParam).Result
   )
   > n20 = &RigUnit_SphericalPoseReader(
    ActiveRegionScaleFactors=(PositiveWidth=1,NegativeWidth=3,PositiveHeight=3,NegativeHeight=3)
    ActiveRegionSize=0.05
    Debug=(bDrawDebug=false,bDraw2D=false,DebugScale=25,DebugSegments=20,DebugThickness=0.25)
    DriverAxis=(X=1,Y=0,Z=0)
    DriverItem=(Type=Bone,Name="thigh_r")
    FalloffRegionScaleFactors=(PositiveWidth=1,NegativeWidth=1,PositiveHeight=1,NegativeHeight=1)
    FalloffSize=0.4
    FlipHeightScaling=false
    OptionalParentItem=(Type=Bone)
    RotationOffset=(X=-82.1368,Y=-4.5092,Z=12.8014)
   )
   > &RigUnit_ModifyTransforms(
    ItemToModify=((Item=(Type=Bone,Name="dyn_parent_cloth_front"),Transform=(Rotation=(X=0,Y=-0.2588,Z=0,W=0.9659),Translation=(X=0,Y=0,Z=0),Scale3D=(X=1,Y=1,Z=1))))
    WeightMinimum=0
    Weight=&RigVMFunction_MathDoubleMul(B=@n20.OutputParam, A=@n18.Value).Result
   )
   RigVMFunction_MathVectorDistance = &RigVMFunction_MathVectorDistance(
    A=&RigUnit_GetTransform(Item=(Type=Bone,Name="crown_r"), bInitial=False).Transform.Translation
    B=&RigUnit_GetTransform(Item=(Type=Bone,Name="head"), bInitial=False).Transform.Translation
   )
   &RigUnit_ModifyTransforms(
    ItemToModify=((Item=(Type=Bone,Name="crown_r"),Transform=(Rotation=(X=0.507,Y=0,Z=0,W=0.8619),Translation=(X=-7.7693,Y=9.2072,Z=3.2057),Scale3D=(X=1,Y=1,Z=1))))
    WeightMinimum=0
    Weight=&RigVMFunction_MathDoubleRemap(SourceMaximum=13, SourceMinimum=24, TargetMaximum=1, TargetMinimum=0, Value=@RigVMFunction_MathVectorDistance.Result).Result
    ExecutePin=@RigVMFunction_Sequence.E
   )
   n21 = &RigUnit_GetTransform(Item=(Type=Bone,Name="crown_l"), bInitial=False)
   n22 = &RigUnit_GetTransform(Item=(Type=Bone,Name="head"), bInitial=False)
   > &RigUnit_ModifyTransforms(
    ItemToModify=((Item=(Type=Bone,Name="crown_l"),Transform=(Rotation=(X=0.5811,Y=0,Z=0,W=0.8138),Translation=(X=5.6774,Y=-9.7421,Z=-3.1614),Scale3D=(X=1,Y=1,Z=1))))
    WeightMinimum=0
    Weight=&RigVMFunction_MathDoubleRemap(
     SourceMaximum=12
     SourceMinimum=25
     TargetMaximum=1
     TargetMinimum=0
     Value=&RigVMFunction_MathVectorDistance(A=@n21.Transform.Translation, B=@n22.Transform.Translation).Result
    ).Result
   )
   n23 = &RigVMFunction_Sequence(ExecuteContext=@RigVMFunction_Sequence.F)
   RigVMFunction_ControlFlowBranch = &RigVMFunction_ControlFlowBranch(ExecuteContext=@n23.A, Condition=RigVMVariableNode() {variable=IsSkyDive}.Value)
   Get Delta Ref = RigVMFunctionReferenceNode(Item=(Type=Bone,Name="pelvis"), ExecuteContext=@RigVMFunction_ControlFlowBranch.False)
   n24 = &RigVMFunction_ControlFlowBranch(ExecuteContext=@RigVMFunction_ControlFlowBranch.True, Condition=RigVMVariableNode() {variable=IsALT}.Value)
   &RigUnit_ModifyTransforms(
    ItemToModify=((Item=(Type=Bone,Name="dyn_driver_cloth_front"),Transform=(Rotation=(X=0,Y=0.6884,Z=0,W=0.7254),Translation=(X=0,Y=0,Z=0),Scale3D=(X=1,Y=1,Z=1))))
    Weight=1
    WeightMinimum=0
    ExecutePin=@n24.False
   )
   n25 = &RigVMFunction_ControlFlowBranch(ExecuteContext=@n23.B, Condition=RigVMVariableNode() {variable=IsALT}.Value)
   n26 = &RigVMFunction_ControlFlowBranch(ExecuteContext=@n25.False, Condition=RigVMVariableNode() {variable=isFloating}.Value)
   &RigUnit_ModifyTransforms(
    ItemToModify=((Item=(Type=Bone,Name="dyn_driver_cloth_front"),Transform=(Rotation=(X=0,Y=0.5373,Z=0,W=0.8434),Translation=(X=0,Y=0,Z=0),Scale3D=(X=1,Y=1,Z=1))))
    Weight=1
    WeightMinimum=0
    ExecutePin=@n26.True
   )
   n27 = &RigVMFunction_ControlFlowBranch(ExecuteContext=@RigVMFunction_Sequence.G, Condition=RigVMVariableNode() {variable=IsDBNO}.Value)
   &RigUnit_ModifyTransforms(
    ItemToModify="((Item=(Type=Bone,Name=\"driver_skirt_rear_01_r\"),Transform=(Rotation=(X=0,Y=0,Z=0.1301,W=0.9915),Translation=(X=10,Y=4,Z=0),Scale3D=(X=1,Y=1,Z=1))),(Item=(Type=Bone,Name=\"driver_skirt_rear_01_l\"),Transform=(Rotation=(X=0,Y=0,Z=0.5736,W=0.8192),Translation= ...(truncated)"
    Weight=1
    WeightMinimum=0
    ExecutePin=@n27.True
   )
   Get Delta Ref_1 = RigVMFunctionReferenceNode(Item=(Type=Bone,Name="foot_r"), ExecuteContext=@RigVMFunction_Sequence_2.B)
   > &RigUnit_ModifyTransforms(
    ItemToModify=((Item=(Type=Bone,Name="leg_harrow_r"),Transform=(Rotation=(X=0,Y=0,Z=-0.2962,W=0.9551),Translation=(X=0,Y=0,Z=0),Scale3D=(X=1,Y=1,Z=1))))
    WeightMinimum=0
    Weight=&RigVMFunction_MathDoubleRemap(SourceMaximum=-25, SourceMinimum=0, TargetMaximum=1, TargetMinimum=0, Value=@"Get Delta Ref_1".Result.Rotation.Yaw).Result
   )
   RigVMFunction_MathQuaternionToEuler = &RigVMFunction_MathQuaternionToEuler(
    RotationOrder=ZYX
    Value=&RigUnit_GetRelativeTransformForItem(Child=(Type=Bone,Name="foot_l"), Parent=(Type=Bone,Name="calf_l"), bChildInitial=False, bParentInitial=False).RelativeTransform.Rotation
   )
   > &RigUnit_ModifyTransforms(
    ItemToModify=((Item=(Type=Bone,Name="leg_harrow_l"),Transform=(Rotation=(X=0,Y=0,Z=-0.2924,W=0.9563),Translation=(X=0,Y=0,Z=0),Scale3D=(X=1,Y=1,Z=1))))
    WeightMinimum=0
    Weight=&RigVMFunction_MathDoubleRemap(SourceMaximum=-40, SourceMinimum=0, TargetMaximum=1, TargetMinimum=0, Value=@RigVMFunction_MathQuaternionToEuler.Result.Z).Result
   )
   n28 = &RigVMFunction_ControlFlowBranch(ExecuteContext=@RigVMFunction_Sequence_2.C, Condition=RigVMVariableNode() {variable=IsALT}.Value)
   n29 = &RigVMFunction_ControlFlowBranch(ExecuteContext=@n28.False, Condition=RigVMVariableNode() {variable=Is parachute opened}.Value)
   &RigUnit_ModifyTransforms(
    ItemToModify="((Item=(Type=Bone,Name=\"dyn_driver_cloth_front\"),Transform=(Rotation=(X=0.1033,Y=0.7014,Z=0.1929,W=0.6784),Translation=(X=-2.2292,Y=-0.9824,Z=-1.0127),Scale3D=(X=1,Y=1,Z=1))),(Item=(Type=Bone,Name=\"driver_skirt_rear_01_l\"),Transform=(Rotation=(X=0,Y=0,Z=-0 ...(truncated)"
    WeightMinimum=0
    Weight=&RigVMFunction_MathDoubleRemap(
     SourceMaximum=25
     SourceMinimum=39
     TargetMaximum=1
     TargetMinimum=0
     Value=&RigUnit_GetRelativeTransformForItem(Child=(Type=Bone,Name="pelvis"), Parent=(Type=Bone,Name="root"), bChildInitial=False, bParentInitial=False).RelativeTransform.Translation.Z
    ).Result
    ExecutePin=@n29.False
   )
   > &RigUnit_ModifyTransforms(
    ItemToModify="((Item=(Type=Bone,Name=\"driver_skirt_rear_01_l\"),Transform=(Rotation=(X=0,Y=0,Z=0,W=1),Translation=(X=0,Y=-3,Z=0),Scale3D=(X=1,Y=1,Z=1))),(Item=(Type=Bone,Name=\"driver_skirt_rear_01_r\"),Transform=(Rotation=(X=-0.2262,Y=0,Z=0,W=0.9741),Translation=(X=0,Y=-1 ...(truncated)"
    WeightMinimum=0
    Weight=&RigVMFunction_MathDoubleRemap(
     SourceMaximum=12
     SourceMinimum=20
     TargetMaximum=1
     TargetMinimum=0
     Value=&RigUnit_GetRelativeTransformForItem(Child=(Type=Bone,Name="pelvis"), Parent=(Type=Bone,Name="root"), bChildInitial=False, bParentInitial=False).RelativeTransform.Translation.Z
    ).Result
   )
   n30 = &RigUnit_GetRelativeTransformForItem(Child=(Type=Bone,Name="hand_r"), Parent=(Type=Bone,Name="lowerarm_r"), bChildInitial=False, bParentInitial=False)
   &RigUnit_ModifyTransforms(
    ItemToModify=((Item=(Type=Bone,Name="bracelet"),Transform=(Rotation=(X=0,Y=0,Z=0,W=1),Translation=(X=1.6136,Y=0,Z=0),Scale3D=(X=1,Y=1,Z=1))))
    WeightMinimum=0
    Weight=&RigVMFunction_MathDoubleRemap(
     SourceMaximum=50
     SourceMinimum=17
     TargetMaximum=1
     TargetMinimum=0
     Value=&RigVMFunction_MathQuaternionToEuler(RotationOrder=ZYX, Value=@n30.RelativeTransform.Rotation).Result.Y
    ).Result
    ExecutePin=@RigVMFunction_Sequence_2.D
   )
   n31 = &RigUnit_GetRelativeTransformForItem(Child=(Type=Bone,Name="hand_r"), Parent=(Type=Bone,Name="lowerarm_r"), bChildInitial=False, bParentInitial=False)
   > &RigUnit_ModifyTransforms(
    ItemToModify=((Item=(Type=Bone,Name="bracelet"),Transform=(Rotation=(X=0,Y=0,Z=0,W=1),Translation=(X=1.9476,Y=0,Z=0),Scale3D=(X=1,Y=1,Z=1))))
    WeightMinimum=0
    Weight=&RigVMFunction_MathDoubleRemap(
     SourceMaximum=-60
     SourceMinimum=-20
     TargetMaximum=1
     TargetMinimum=0
     Value=&RigVMFunction_MathQuaternionToEuler(RotationOrder=ZYX, Value=@n31.RelativeTransform.Rotation).Result.Y
    ).Result
   )
   n32 = &RigUnit_GetRelativeTransformForItem(Child=(Type=Bone,Name="calf_l"), Parent=(Type=Bone,Name="thigh_l"), bChildInitial=False, bParentInitial=False)
   &RigUnit_ModifyTransforms(
    ItemToModify="((Item=(Type=Bone,Name=\"Base_feathers_1_l\"),Transform=(Rotation=(X=0.0245,Y=-0.4691,Z=-0.1715,W=0.866),Translation=(X=0,Y=0,Z=0),Scale3D=(X=1,Y=1,Z=1))),(Item=(Type=Bone,Name=\"Base_feathers_2_l\"),Transform=(Rotation=(X=0,Y=-0.4019,Z=0,W=0.9157),Translation ...(truncated)"
    WeightMinimum=0
    Weight=&RigVMFunction_MathDoubleRemap(
     SourceMaximum=120
     SourceMinimum=50
     TargetMaximum=1
     TargetMinimum=0
     Value=&RigVMFunction_MathQuaternionToEuler(RotationOrder=ZYX, Value=@n32.RelativeTransform.Rotation).Result.Z
    ).Result
    ExecutePin=@RigVMFunction_Sequence_2.E
   )
   n33 = &RigUnit_GetRelativeTransformForItem(Child=(Type=Bone,Name="calf_r"), Parent=(Type=Bone,Name="thigh_r"), bChildInitial=False, bParentInitial=False)
   > &RigUnit_ModifyTransforms(
    ItemToModify="((Item=(Type=Bone,Name=\"Base_feathers_1_r\"),Transform=(Rotation=(X=0,Y=-0.231,Z=0,W=0.9729),Translation=(X=0,Y=0,Z=0),Scale3D=(X=1,Y=1,Z=1))),(Item=(Type=Bone,Name=\"Base_feathers_2_r\"),Transform=(Rotation=(X=0.0562,Y=-0.329,Z=-0.1588,W=0.9292),Translation= ...(truncated)"
    WeightMinimum=0
    Weight=&RigVMFunction_MathDoubleRemap(
     SourceMaximum=120
     SourceMinimum=50
     TargetMaximum=1
     TargetMinimum=0
     Value=&RigVMFunction_MathQuaternionToEuler(RotationOrder=ZYX, Value=@n33.RelativeTransform.Rotation).Result.Z
    ).Result
   )
   n34 = &RigVMFunction_ControlFlowBranch(ExecuteContext=@RigVMFunction_Sequence_4.B, Condition=RigVMVariableNode() {variable=IsEmote}.Value)
   n35 = &RigUnit_SphericalPoseReader(
    ActiveRegionScaleFactors=(PositiveWidth=0.8,NegativeWidth=0,PositiveHeight=0.2,NegativeHeight=0.5)
    ActiveRegionSize=0.3
    Debug=(bDrawDebug=true,bDraw2D=False,bDrawLocalAxes=False,DebugScale=25,DebugSegments=20,DebugThickness=0.25)
    DriverAxis=(X=1,Y=0,Z=0)
    DriverItem=(Type=Bone,Name="thigh_r")
    FalloffRegionScaleFactors=(PositiveWidth=0.4,NegativeWidth=0,PositiveHeight=0.4,NegativeHeight=0.4)
    FalloffSize=0.5
    FlipHeightScaling=False
    FlipWidthScaling=False
    OptionalParentItem=(Type=Bone,Name="pelvis")
    RotationOffset=(X=-90,Y=0,Z=-60)
    ExecutePin=@n34.True
   )
   > &RigUnit_ModifyTransforms(
    ItemToModify=((Item=(Type=Bone,Name="dyn_driver_cloth_front"),Transform=(Rotation=(X=0,Y=-0.7281,Z=0,W=0.6855),Translation=(X=-9.1522,Y=0,Z=-7.0234),Scale3D=(X=1,Y=1,Z=1))))
    WeightMinimum=0
    Weight=@n35.OutputParam
   )
   n36 = &RigVMFunction_ControlFlowBranch(ExecuteContext=@RigVMFunction_Sequence_4.C, Condition=RigVMVariableNode() {variable=IsALT}.Value)
   n37 = &RigUnit_GetRelativeTransformForItem(Child=(Type=Bone,Name="calf_r"), Parent=(Type=Bone,Name="thigh_r"), bChildInitial=False, bParentInitial=False)
   &RigUnit_ModifyTransforms(
    ItemToModify="((Item=(Type=Bone,Name=\"Base_feathers_1_r\"),Transform=(Rotation=(X=0,Y=0.1253,Z=0,W=0.9921),Translation=(X=0,Y=0,Z=0),Scale3D=(X=1,Y=1,Z=1))),(Item=(Type=Bone,Name=\"Base_feathers_2_r\"),Transform=(Rotation=(X=0,Y=0.0698,Z=0,W=0.9976),Translation=(X=-1.2709, ...(truncated)"
    WeightMinimum=0
    Weight=&RigVMFunction_MathDoubleRemap(
     SourceMaximum=120
     SourceMinimum=50
     TargetMaximum=1
     TargetMinimum=0
     Value=&RigVMFunction_MathQuaternionToEuler(RotationOrder=ZYX, Value=@n37.RelativeTransform.Rotation).Result.Z
    ).Result
    ExecutePin=@n36.True
   )
   > &RigUnit_ModifyTransforms(
    ItemToModify=((Item=(Type=Bone,Name="AltSkin_corr_back"),Transform=(Rotation=(X=0.2034,Y=0,Z=0,W=0.9791),Translation=(X=0,Y=0,Z=0),Scale3D=(X=1,Y=1,Z=1))))
    WeightMinimum=0
    Weight=RigVMVariableNode() {variable=Crouch}.Value
   )
   > &RigUnit_ModifyTransforms(
    ItemToModify=((Item=(Type=Bone,Name="AltSkin_corr_back"),Transform=(Rotation=(X=-0.1736,Y=0,Z=0,W=0.9848),Translation=(X=0,Y=0,Z=0),Scale3D=(X=1,Y=1,Z=1))))
    WeightMinimum=0
    Weight=RigVMVariableNode() {variable=CrouchWalk}.Value
   )
   Multiply_5 = RigVMAggregateNode(
    A=&RigVMFunction_MathDoubleRemap(
     SourceMaximum=81
     SourceMinimum=80
     TargetMaximum=1
     TargetMinimum=0
     Value=&RigUnit_GetRelativeTransformForItem(Child=(Type=Bone,Name="pelvis"), Parent=(Type=Bone,Name="root"), bChildInitial=False, bParentInitial=False).RelativeTransform.Translation.Z
    ).Result
    C=&RigVMFunction_MathDoubleRemap(SourceMaximum=400, SourceMinimum=250, TargetMaximum=1, TargetMinimum=0, Value=RigVMVariableNode() {variable=Velocity character}.Value).Result
    B=RigVMVariableNode() {variable=ForwardMovement}.Value
   )
   > &RigUnit_ModifyTransforms(
    ItemToModify=((Item=(Type=Bone,Name="AltSkin_corr_back"),Transform=(Rotation=(X=-0.3827,Y=0,Z=0,W=0.9239),Translation=(X=0,Y=0,Z=0),Scale3D=(X=1,Y=1,Z=1))))
    WeightMinimum=0
    Weight=&RigVMFunction_KalmanFloat(BufferSize=8, Value=@Multiply_5.Result) {template="KalmanFilter::Execute(in Value,in BufferSize,out Result)"}.Result
   )
   Multiply_6 = RigVMAggregateNode(
    A=&RigVMFunction_MathDoubleRemap(
     SourceMaximum=81
     SourceMinimum=80
     TargetMaximum=1
     TargetMinimum=0
     Value=&RigUnit_GetRelativeTransformForItem(Child=(Type=Bone,Name="pelvis"), Parent=(Type=Bone,Name="root"), bChildInitial=False, bParentInitial=False).RelativeTransform.Translation.Z
    ).Result
    C=&RigVMFunction_MathDoubleRemap(SourceMaximum=600, SourceMinimum=550, TargetMaximum=1, TargetMinimum=0, Value=RigVMVariableNode() {variable=Velocity character}.Value).Result
    B=RigVMVariableNode() {variable=ForwardMovement}.Value
   )
   > &RigUnit_ModifyTransforms(
    ItemToModify=((Item=(Type=Bone,Name="AltSkin_corr_back"),Transform=(Rotation=(X=-0.1736,Y=0,Z=0,W=0.9848),Translation=(X=0,Y=0,Z=0),Scale3D=(X=1,Y=1,Z=1))))
    WeightMinimum=0
    Weight=&RigVMFunction_KalmanFloat(BufferSize=8, Value=@Multiply_6.Result) {template="KalmanFilter::Execute(in Value,in BufferSize,out Result)"}.Result
   )
   &RigUnit_ModifyTransforms(
    ExecutePin=()
    ItemToModify="((Item=(Type=Bone,Name=\"crown_l\"),Transform=(Rotation=(X=0.3417,Y=0,Z=0,W=0.9398),Translation=(X=4.0093,Y=-7.2324,Z=-2.4539),Scale3D=(X=1,Y=1,Z=1))),(Item=(Type=Bone,Name=\"crown_r\"),Transform=(Rotation=(X=0.2433,Y=0,Z=0,W=0.97),Translation=(X=-5.1913,Y=4.7 ...(truncated)"
    WeightMinimum=0
    Weight=&RigVMFunction_KalmanFloat(
     BufferSize=32
     Value=&RigVMFunction_MathDoubleRemap(SourceMaximum=80, SourceMinimum=60, TargetMaximum=1, TargetMinimum=0, Value=@"Get Delta Ref".Result.Rotation.Yaw).Result
    ) {template="KalmanFilter::Execute(in Value,in BufferSize,out Result)"}.Result
   )
   &RigVMFunction_ControlFlowBranch(Condition=False, ExecuteContext=())
  ]}
]
hierarchy [
 root {loc=<0 0 0>, rot=<0 0 0 1>}
 pelvis {parent=root, loc=<0 -0.5301 98.0571>, rot=<0.0127 0.707 -0.0127 -0.707>}
 spine_01 {parent=pelvis, loc=<6.1208 -0.9328 0>, rot=<0 0 0.034 -0.9994>}
 spine_02 {parent=spine_01, loc=<6.5997 -0.2095 0>, rot=<0 0 0.0238 -0.9997>}
 spine_03 {parent=spine_02, loc=<6.4914 -0.1237 0>, rot=<0 0 0.0149 -0.9999>}
 spine_04 {parent=spine_03, loc=<6.9598 0.0794 0>, rot=<0 0 -0.0132 -0.9999>}
 spine_05 {parent=spine_04, loc=<6.9987 0 0>, rot=<0 0 -0.0015 -1>}
 neck_01 {parent=spine_05, loc=<11.3138 0.1529 0>, rot=<0 0 -0.2003 -0.9797>}
 neck_02 {parent=neck_01, loc=<4.5627 0 0>, rot=<0 0 0.0504 -0.9987>}
 head {parent=neck_02, loc=<5.3914 0 0>, rot=<0 0 0.115 -0.9934>}
 clavicle_l {parent=spine_05, loc=<5.4831 -0.1053 -3.2946>, rot=<0.0465 -0.7289 -0.009 -0.6829>}
 upperarm_l {parent=clavicle_l, loc=<9.8658 -1.2136 -0.4105>, rot=<0.0164 -0.3245 -0.0269 -0.9454>}
 lowerarm_l {parent=upperarm_l, loc=<26.465 -0.8345 0.259>, rot=<0.0036 0.0161 -0.2298 -0.9731>}
 hand_l {parent=lowerarm_l, loc=<22.7355 -0.3203 -0.6105>, rot=<-0.6784 -0.0223 0.0735 -0.7307>}
 thumb_01_l {parent=hand_l, loc=<1.3118 -0.3191 -2.459>, rot=<0.6546 -0.4042 0.0664 -0.6354>}
 thumb_02_l {parent=thumb_01_l, loc=<5.9336 -0.0448 0.0401>, rot=<-0.0037 0.0213 0.298 -0.9543>}
 thumb_03_l {parent=thumb_02_l, loc=<2.2223 0.3682 -0.0907>, rot=<0.0118 -0.013 0.0066 -0.9998>}
 index_metacarpal_l {parent=hand_l, loc=<1.6535 0.0183 -2.96>, rot=<0.0158 -0.0124 -0.0173 -0.9996>}
 index_01_l {parent=index_metacarpal_l, loc=<8.3008 0.0357 -0.0524>, rot=<-0.0044 0.0485 0.0924 -0.9945>}
 index_02_l {parent=index_01_l, loc=<3.3587 -0.3004 0.0452>, rot=<-0.002 0.0027 0.3225 -0.9466>}
 index_03_l {parent=index_02_l, loc=<2.5578 -0.123 -0.0252>, rot=<-0.002 0.0067 0.2558 -0.9667>}
 middle_metacarpal_l {parent=hand_l, loc=<2.1856 -0.1725 -0.5984>, rot=<0.006 0.0168 -0.0329 -0.9993>}
 middle_01_l {parent=middle_metacarpal_l, loc=<7.3275 -0.2447 0.0129>, rot=<-0.0208 0.0042 0.0858 -0.9961>}
 middle_02_l {parent=middle_01_l, loc=<4.0123 0.1401 0.1175>, rot=<-0.0218 0.0258 0.3512 -0.9357>}
 middle_03_l {parent=middle_02_l, loc=<2.6077 0.0534 0.1373>, rot=<0.0082 0.0188 0.1708 -0.9851>}
 ring_metacarpal_l {parent=hand_l, loc=<2.2074 -0.3625 1.3638>, rot=<-0.0284 0.0722 -0.0458 -0.9959>}
 ring_01_l {parent=ring_metacarpal_l, loc=<7.0295 -0.6197 -0.1446>, rot=<-0.0489 -0.0252 0.1892 -0.9804>}
 ring_02_l {parent=ring_01_l, loc=<3.6174 0.4647 -0.1027>, rot=<0 0 0.2806 -0.9598>}
 ring_03_l {parent=ring_02_l, loc=<2.4152 0.114 0.0199>, rot=<0 0 0.1189 -0.9929>}
 pinky_metacarpal_l {parent=hand_l, loc=<1.7616 -0.9753 3.3814>, rot=<-0.0397 0.0814 -0.069 -0.9935>}
 pinky_01_l {parent=pinky_metacarpal_l, loc=<7.1709 -0.4831 -0.1094>, rot=<-0.0654 -0.0035 0.2242 -0.9723>}
 pinky_02_l {parent=pinky_01_l, loc=<2.8274 0.0893 -0.1043>, rot=<0.0037 0.0011 0.2854 -0.9584>}
 pinky_03_l {parent=pinky_02_l, loc=<2.2046 0.0307 -0.0078>, rot=<-0.0002 0.0098 0.1445 -0.9895>}
 deform_wrist_dn_base_l {parent=hand_l, loc=<0 -1.5326 0>, rot=<0 0 0 -1>}
 deform_wrist_dn_l {parent=deform_wrist_dn_base_l, loc=<0 0 0>, rot=<0 0 0 1>}
 deform_wrist_up_base_l {parent=hand_l, loc=<0 2.3009 0>, rot=<0 0 0 -1>}
 deform_wrist_up_l {parent=deform_wrist_up_base_l, loc=<0 0 0>, rot=<0 0 0 1>}
 lowerarm_twist_02_l {parent=lowerarm_l, loc=<10.6707 0 0>, rot=<0 0 0 1>}
 lowerarm_twist_01_l {parent=lowerarm_l, loc=<18.316 0 0>, rot=<0 0 0 1>}
 deform_elbow_in_base_l {parent=lowerarm_l, loc=<0.1834 3.6062 -0.2572>, rot=<0 0 0 -1>}
 deform_elbow_in_l {parent=deform_elbow_in_base_l, loc=<0 0 0>, rot=<0 0 0 1>}
 deform_elbow_out_base_l {parent=lowerarm_l, loc=<0 -4 0>, rot=<0 0 0 1>}
 deform_elbow_out_l {parent=deform_elbow_out_base_l, loc=<0 0 0>, rot=<0 0 0 1>}
 upperarm_twist_01_l {parent=upperarm_l, loc=<7.2135 -0.1134 0>, rot=<0 0 0 1>}
 deform_upperarm_bk_base_l {parent=upperarm_twist_01_l, loc=<-6.4053 -4.3746 0.314>, rot=<0 0 0 -1>}
 deform_upperarm_bk_l {parent=deform_upperarm_bk_base_l, loc=<0 0 0>, rot=<0 0 0 1>}
 deform_upperarm_fr_base_l {parent=upperarm_twist_01_l, loc=<-5.6312 5.092 0.3354>, rot=<0 0 0 -1>}
 deform_upperarm_fr_l {parent=deform_upperarm_fr_base_l, loc=<0 0 0>, rot=<0 0 0 1>}
 deform_upperarm_up_base_l {parent=upperarm_twist_01_l, loc=<-8.891 0.4825 3.9517>, rot=<0 0 0 -1>}
 deform_upperarm_up_l {parent=deform_upperarm_up_base_l, loc=<0 0 0>, rot=<0 0 0 1>}
 crown_l {parent=upperarm_twist_01_l, loc=<-5.535 0.4825 7.318>, rot=<-0.0075 -0.0283 0.0011 -0.9996>}
 upperarm_twist_02_l {parent=upperarm_l, loc=<19.7998 -0.1134 0>, rot=<0 0 0 1>}
 clavicle_r {parent=spine_05, loc=<5.4831 -0.1053 3.2946>, rot=<0.7289 0.0465 -0.6829 0.009>}
 upperarm_r {parent=clavicle_r, loc=<-9.8658 1.2136 0.4105>, rot=<0.0164 -0.3245 -0.0269 -0.9454>}
 lowerarm_r {parent=upperarm_r, loc=<-26.465 0.8345 -0.259>, rot=<0.0036 0.0161 -0.2298 -0.9731>}
 hand_r {parent=lowerarm_r, loc=<-22.7355 0.3203 0.6106>, rot=<-0.6784 -0.0223 0.0735 -0.7307>}
 thumb_01_r {parent=hand_r, loc=<-1.3118 0.3191 2.459>, rot=<0.6546 -0.4042 0.0664 -0.6354>}
 thumb_02_r {parent=thumb_01_r, loc=<-5.9336 0.0448 -0.0401>, rot=<-0.0037 0.0213 0.298 -0.9543>}
 thumb_03_r {parent=thumb_02_r, loc=<-2.2223 -0.3682 0.0907>, rot=<0.0118 -0.013 0.0066 -0.9998>}
 index_metacarpal_r {parent=hand_r, loc=<-1.6535 -0.0183 2.96>, rot=<0.0158 -0.0124 -0.0173 -0.9996>}
 index_01_r {parent=index_metacarpal_r, loc=<-8.3008 -0.0357 0.0524>, rot=<-0.0044 0.0485 0.0924 -0.9945>}
 index_02_r {parent=index_01_r, loc=<-3.3587 0.3004 -0.0452>, rot=<-0.002 0.0027 0.3225 -0.9466>}
 index_03_r {parent=index_02_r, loc=<-2.5578 0.123 0.0252>, rot=<-0.002 0.0067 0.2558 -0.9667>}
 middle_metacarpal_r {parent=hand_r, loc=<-2.1856 0.1725 0.5984>, rot=<0.006 0.0168 -0.0329 -0.9993>}
 middle_01_r {parent=middle_metacarpal_r, loc=<-7.3275 0.2447 -0.0129>, rot=<-0.0208 0.0042 0.0858 -0.9961>}
 middle_02_r {parent=middle_01_r, loc=<-4.0123 -0.1401 -0.1175>, rot=<-0.0218 0.0258 0.3512 -0.9357>}
 middle_03_r {parent=middle_02_r, loc=<-2.6077 -0.0534 -0.1373>, rot=<0.0082 0.0188 0.1708 -0.9851>}
 ring_metacarpal_r {parent=hand_r, loc=<-2.2074 0.3625 -1.3638>, rot=<-0.0284 0.0722 -0.0458 -0.9959>}
 ring_01_r {parent=ring_metacarpal_r, loc=<-7.0295 0.6197 0.1446>, rot=<-0.0489 -0.0252 0.1892 -0.9804>}
 ring_02_r {parent=ring_01_r, loc=<-3.6174 -0.4647 0.1027>, rot=<0 0 0.2806 -0.9598>}
 ring_03_r {parent=ring_02_r, loc=<-2.4152 -0.114 -0.0199>, rot=<0 0 0.1189 -0.9929>}
 pinky_metacarpal_r {parent=hand_r, loc=<-1.7616 0.9753 -3.3814>, rot=<-0.0397 0.0814 -0.069 -0.9935>}
 pinky_01_r {parent=pinky_metacarpal_r, loc=<-7.1709 0.4831 0.1094>, rot=<-0.0654 -0.0035 0.2242 -0.9723>}
 pinky_02_r {parent=pinky_01_r, loc=<-2.8274 -0.0893 0.1043>, rot=<0.0037 0.0011 0.2854 -0.9584>}
 pinky_03_r {parent=pinky_02_r, loc=<-2.2046 -0.0307 0.0078>, rot=<-0.0002 0.0098 0.1445 -0.9895>}
 deform_wrist_dn_base_r {parent=hand_r, loc=<-0.0001 1.5328 0>, rot=<0 0 0 -1>}
 deform_wrist_dn_r {parent=deform_wrist_dn_base_r, loc=<0 0 0>, rot=<0 0 0 1>}
 deform_wrist_up_base_r {parent=hand_r, loc=<-0.0002 -2.3007 -0.0001>, rot=<0 0 0 -1>}
 deform_wrist_up_r {parent=deform_wrist_up_base_r, loc=<0 0 0>, rot=<0 0 0 1>}
 lowerarm_twist_02_r {parent=lowerarm_r, loc=<-10.6707 0 0>, rot=<0 0 0 -1>}
 lowerarm_twist_01_r {parent=lowerarm_r, loc=<-18.316 0 0>, rot=<0 0 0 -1>}
 deform_elbow_in_base_r {parent=lowerarm_r, loc=<-0.1834 -3.6062 0.2572>, rot=<0 0 0 -1>}
 deform_elbow_in_r {parent=deform_elbow_in_base_r, loc=<0 0 0>, rot=<0 0 0 1>}
 deform_elbow_out_base_r {parent=lowerarm_r, loc=<0 4 -0.0001>, rot=<0 0 0 -1>}
 deform_elbow_out_r {parent=deform_elbow_out_base_r, loc=<0 0 0>, rot=<0 0 0 1>}
 bracelet {parent=lowerarm_r, loc=<-20.5943 0 0>, rot=<0 0 0 -1>}
 upperarm_twist_01_r {parent=upperarm_r, loc=<-7.2135 0.1134 0>, rot=<0 0 0 -1>}
 deform_upperarm_bk_base_r {parent=upperarm_twist_01_r, loc=<6.4053 4.3746 -0.314>, rot=<0 0 0 -1>}
 deform_upperarm_bk_r {parent=deform_upperarm_bk_base_r, loc=<0 0 0>, rot=<0 0 0 1>}
 deform_upperarm_fr_base_r {parent=upperarm_twist_01_r, loc=<5.6312 -5.092 -0.3354>, rot=<0 0 0 -1>}
 deform_upperarm_fr_r {parent=deform_upperarm_fr_base_r, loc=<0 0 0>, rot=<0 0 0 1>}
 deform_upperarm_up_base_r {parent=upperarm_twist_01_r, loc=<8.891 -0.4825 -3.9517>, rot=<0 0 0 -1>}
 deform_upperarm_up_r {parent=deform_upperarm_up_base_r, loc=<0 0 0>, rot=<0 0 0 1>}
 crown_r {parent=upperarm_twist_01_r, loc=<5.535 -0.482 -7.318>, rot=<0 0 0 1>}
 upperarm_twist_02_r {parent=upperarm_r, loc=<-19.7998 0.1134 0>, rot=<0 0 0 -1>}
 attach_backpack {parent=spine_05, loc=<-0.1218 -13.4733 0>, rot=<-0.0093 -0.707 0.0093 -0.707>}
 attach_cape {parent=spine_05, loc=<9.0178 -11.7142 0>, rot=<0.0616 -0.7044 -0.0616 -0.7044>}
 deform_pec_aim_base_l {parent=spine_04, loc=<3.2456 9.3309 -1.2237>, rot=<0.0603 -0.4033 0.3318 -0.8507>}
 deform_pec_aim_l {parent=deform_pec_aim_base_l, loc=<0 0 0>, rot=<0 0 0 1>}
 deform_pec_base_l {parent=deform_pec_aim_l, loc=<9.3665 0 0>, rot=<0 0 0 1>}
 deform_pec_l {parent=deform_pec_base_l, loc=<0 0 0>, rot=<0 0 0 1>}
 deform_pec_aim_base_r {parent=spine_04, loc=<3.2452 9.3309 1.2237>, rot=<-0.4033 -0.0603 0.8507 0.3318>}
 deform_pec_aim_r {parent=deform_pec_aim_base_r, loc=<0 0 0>, rot=<0 0 0 1>}
 deform_pec_base_r {parent=deform_pec_aim_r, loc=<-9.3667 -0.0003 -0.0004>, rot=<0 0 0 1>}
 deform_pec_r {parent=deform_pec_base_r, loc=<0 0 0>, rot=<0 0 0 1>}
 dyn_parent_cloth_front {parent=spine_01, loc=<2.5928 9.3325 0>, rot=<0.0112 0.707 0.707 0.0112>}
 dyn_driver_cloth_front {parent=dyn_parent_cloth_front, loc=<0 0 0>, rot=<0 0 0 1>}
 dyn_cloth_font_mid_01 {parent=dyn_driver_cloth_front, loc=<3.8599 0 0>, rot=<0 0 0 1>}
 dyn_loincloth_01_r {parent=dyn_cloth_font_mid_01, loc=<6.5483 5.7946 -0.0287>, rot=<0 -0.9996 0 -0.0265>}
 dyn_loincloth_02_r {parent=dyn_loincloth_01_r, loc=<-18.7843 0 0.237>, rot=<0 0.0183 0 -0.9998>}
 dyn_loincloth_03_r {parent=dyn_loincloth_02_r, loc=<-14.0721 0 0>, rot=<0 0 0 1>}
 dyn_loincloth_01_l {parent=dyn_cloth_font_mid_01, loc=<6.5482 -5.7946 -0.0287>, rot=<0 0.0265 0 -0.9996>}
 dyn_loincloth_02_l {parent=dyn_loincloth_01_l, loc=<18.7843 0 -0.2369>, rot=<0 0.0183 0 -0.9998>}
 dyn_loincloth_03_l {parent=dyn_loincloth_02_l, loc=<14.0721 0 0>, rot=<0 0 0 1>}
 thigh_l {parent=pelvis, loc=<-1.0855 -0.2109 -8.7053>, rot=<0.0009 0.0355 0.0155 -0.9993>}
 calf_l {parent=thigh_l, loc=<-42.5284 -0.0019 -0.7297>, rot=<0.0006 -0.0161 -0.0391 -0.9991>}
 foot_l {parent=calf_l, loc=<-46.7401 -2.5014 -0.592>, rot=<0.0427 -0.0238 0.0353 -0.9982>}
 ball_l {parent=foot_l, loc=<-3.5658 10.8066 -0.1544>, rot=<0.0001 -0.0001 -0.7063 -0.7079>}
 calf_twist_02_l {parent=calf_l, loc=<-19.1546 -1.7943 -0.2638>, rot=<0.0237 0.0095 -0.0035 -0.9997>}
 calf_twist_01_l {parent=calf_l, loc=<-35.8483 -1.9197 -0.5794>, rot=<0.0237 0.0095 -0.0035 -0.9997>}
 deform_knee_in_base_l {parent=calf_l, loc=<0.8031 -5.0231 0.2456>, rot=<0 0 0 -1>}
 deform_knee_in_l {parent=deform_knee_in_base_l, loc=<0 0 0>, rot=<0 0 0 1>}
 deform_knee_out_base_l {parent=calf_l, loc=<0 4 0>, rot=<0 0 0 -1>}
 deform_knee_out_l {parent=deform_knee_out_base_l, loc=<0 0 0>, rot=<0 0 0 1>}
 leg_harrow_l {parent=calf_l, loc=<-40.2765 1.5373 -1.26>, rot=<0.0427 -0.0238 0.0353 -0.9982>}
 Base_feathers_1_l {parent=calf_l, loc=<-11.1909 -10.1425 -1.4252>, rot=<-0.2275 0.5606 0.7798 -0.1609>}
 feathers_1_l {parent=Base_feathers_1_l, loc=<0 0 0>, rot=<0 0 0 1>}
 Base_feathers_2_l {parent=calf_l, loc=<-11.7518 -7.084 -6.9107>, rot=<-0.2653 0.3621 0.8906 -0.0728>}
 feathers_2_l {parent=Base_feathers_2_l, loc=<0 0 0>, rot=<0 0 0 1>}
 Base_feathers_3_l {parent=calf_l, loc=<-13.1993 -1.1384 -8.2792>, rot=<-0.2913 -0.1181 0.9479 0.0515>}
 feathers_3_l {parent=Base_feathers_3_l, loc=<0 0 0>, rot=<0 0 0 1>}
 thigh_twist_01_l {parent=thigh_l, loc=<-27.952 0 0>, rot=<0.0013 0 0 -1>}
 Base_skirt_frontMid_r {parent=thigh_twist_01_l, loc=<27.952 0 0>, rot=<-0.0024 -0.9994 -0.0015 0.0355>}
 dyn_skirt_side_01_r {parent=Base_skirt_frontMid_r, loc=<-2.0749 1.7753 6.1512>, rot=<0.9946 0.0067 0.1036 -0.0015>}
 Driver_skirt_side_02_r1 {parent=dyn_skirt_side_01_r, loc=<13.9131 0 0>, rot=<-0.0003 -0.0012 0 -1>}
 dyn_skirt_side_02_r {parent=Driver_skirt_side_02_r1, loc=<0 0 0>, rot=<-0.0003 -0.0012 0 -1>}
 dyn_skirt_side_03_r {parent=dyn_skirt_side_02_r, loc=<11.4534 0 0>, rot=<0 0 0 1>}
 dyn_skirt_sideMid_01_r {parent=Base_skirt_frontMid_r, loc=<-0.9685 8.3323 1.7351>, rot=<0.8826 -0.0242 0.1134 0.4557>}
 Driver_skirt_sideMid_02_r {parent=dyn_skirt_sideMid_01_r, loc=<13.9131 0 0>, rot=<-0.0003 -0.0012 0 -1>}
 dyn_skirt_sideMid_02_r {parent=Driver_skirt_sideMid_02_r, loc=<0 0 0>, rot=<-0.0003 -0.0012 0 -1>}
 dyn_skirt_sideMid_03_r {parent=dyn_skirt_sideMid_02_r, loc=<11.4534 0 0>, rot=<0 0 0 1>}
 dyn_skirt_frontMid_01_r {parent=Base_skirt_frontMid_r, loc=<-1.3865 -7.3857 3.8102>, rot=<0.9546 -0.0053 0.0604 -0.2916>}
 Driver_skirt_frontMid_02_r {parent=dyn_skirt_frontMid_01_r, loc=<13.417 0 0>, rot=<0 -0.0348 0 -0.9994>}
 dyn_skirt_frontMid_02_r {parent=Driver_skirt_frontMid_02_r, loc=<0 0 0>, rot=<0 -0.0348 0 -0.9994>}
 dyn_skirt_frontMid_03_r {parent=dyn_skirt_frontMid_02_r, loc=<11.6255 0 0>, rot=<0 0 0 1>}
 deform_groin_base_l {parent=thigh_l, loc=<0.4801 5.6776 -0.2486>, rot=<0 0 0 -1>}
 deform_groin_l {parent=deform_groin_base_l, loc=<0 0 0>, rot=<0 0 0 1>}
 thigh_r {parent=pelvis, loc=<-1.0855 -0.2109 8.7053>, rot=<0.0355 -0.0009 0.9993 0.0155>}
 calf_r {parent=thigh_r, loc=<42.5284 0.0019 0.7297>, rot=<0.0006 -0.0161 -0.0391 -0.9991>}
 foot_r {parent=calf_r, loc=<46.7401 2.5014 0.592>, rot=<0.0427 -0.0238 0.0353 -0.9982>}
 ball_r {parent=foot_r, loc=<3.5658 -10.8066 0.1544>, rot=<0.0001 -0.0001 -0.7063 -0.7079>}
 calf_twist_02_r {parent=calf_r, loc=<19.1546 1.7943 0.2638>, rot=<0.0237 0.0095 -0.0035 -0.9997>}
 calf_twist_01_r {parent=calf_r, loc=<35.8483 1.9197 0.5794>, rot=<0.0237 0.0095 -0.0035 -0.9997>}
 deform_knee_in_base_r {parent=calf_r, loc=<-0.8031 5.0231 -0.2457>, rot=<0 0 0 -1>}
 deform_knee_in_r {parent=deform_knee_in_base_r, loc=<0 0 0>, rot=<0 0 0 1>}
 deform_knee_out_base_r {parent=calf_r, loc=<0 -4 0>, rot=<0 0 0 -1>}
 deform_knee_out_r {parent=deform_knee_out_base_r, loc=<0 0 0>, rot=<0 0 0 1>}
 leg_harrow_r {parent=calf_r, loc=<40.2736 -1.5369 1.1856>, rot=<0.0427 -0.0238 0.0353 -0.9982>}
 Base_feathers_1_r {parent=calf_r, loc=<11.1909 10.1425 1.4252>, rot=<-0.7798 -0.1609 -0.2275 -0.5606>}
 feathers_1_r {parent=Base_feathers_1_r, loc=<0 0 0>, rot=<0 0 0 1>}
 Base_feathers_2_r {parent=calf_r, loc=<11.7518 7.084 6.9107>, rot=<0.8906 0.0728 0.2653 0.3621>}
 feathers_2_r {parent=Base_feathers_2_r, loc=<0 0 0>, rot=<0 0 0 1>}
 Base_feathers_3_r {parent=calf_r, loc=<13.1994 1.1384 8.2792>, rot=<0.9479 -0.0515 0.2913 -0.1181>}
 feathers_3_r {parent=Base_feathers_3_r, loc=<0 0 0>, rot=<0 0 0 1>}
 thigh_twist_01_r {parent=thigh_r, loc=<27.952 0 0>, rot=<0.0013 0 0 -1>}
 Base_skirt_frontMid_l {parent=thigh_twist_01_r, loc=<-27.9519 0 0>, rot=<-0.0024 -0.9994 -0.0015 0.0355>}
 dyn_skirt_side_01_l {parent=Base_skirt_frontMid_l, loc=<2.0749 -1.7753 -6.1512>, rot=<0.9946 0.0067 0.1036 -0.0015>}
 dyn_skirt_side_02_l {parent=dyn_skirt_side_01_l, loc=<-13.9131 0 0>, rot=<-0.0003 -0.0012 0 -1>}
 dyn_skirt_side_03_l {parent=dyn_skirt_side_02_l, loc=<-11.4534 0 0>, rot=<0 0 0 1>}
 dyn_skirt_sideMid_03_l1 {parent=Base_skirt_frontMid_l, loc=<0.9685 -8.3323 -1.7351>, rot=<0.8826 -0.0242 0.1134 0.4557>}
 dyn_skirt_sideMid_01_l {parent=dyn_skirt_sideMid_03_l1, loc=<-13.9131 0 0>, rot=<-0.0003 -0.0012 0 -1>}
 dyn_skirt_sideMid_02_l {parent=dyn_skirt_sideMid_01_l, loc=<-11.4534 0 0>, rot=<0 0 0 1>}
 dyn_skirt_sideMid_03_l {parent=Base_skirt_frontMid_l, loc=<1.3866 7.3857 -3.8102>, rot=<0.9546 -0.0053 0.0604 -0.2916>}
 dyn_skirt_frontMid_01_l {parent=dyn_skirt_sideMid_03_l, loc=<-13.417 0 0>, rot=<0 -0.0348 0 -0.9994>}
 dyn_skirt_frontMid_02_l {parent=dyn_skirt_frontMid_01_l, loc=<-11.6255 0 0>, rot=<0 0 0 -1>}
 deform_groin_base_r {parent=thigh_r, loc=<-0.4801 -5.6776 0.2486>, rot=<0 0 0 -1>}
 deform_groin_r {parent=deform_groin_base_r, loc=<0 0 0>, rot=<0 0 0 1>}
 deform_glute_base_l {parent=pelvis, loc=<-7.3466 -9.0213 -9.1245>, rot=<0.0172 0.031 -0.4477 -0.8935>}
 deform_glute_l {parent=deform_glute_base_l, loc=<0 0 0>, rot=<0 0 0 1>}
 deform_glute_base_r {parent=pelvis, loc=<-7.3464 -9.0213 9.1245>, rot=<0.031 -0.0172 0.8935 -0.4477>}
 deform_glute_r {parent=deform_glute_base_r, loc=<0 0 0>, rot=<0 0 0 1>}
 driver_skirt_rear_01_r {parent=pelvis, loc=<-0.255 -11.6668 5.8725>, rot=<-0.0179 -0.9998 0 0>}
 dyn_skirt_rear_01_r {parent=driver_skirt_rear_01_r, loc=<0 0 0>, rot=<0 0 0 1>}
 dyn_skirt_rear_02_r {parent=dyn_skirt_rear_01_r, loc=<15.1161 0 0>, rot=<-0.0079 -0.0006 0.0686 -0.9976>}
 dyn_skirt_rear_03_r {parent=dyn_skirt_rear_02_r, loc=<12.7812 0 0>, rot=<0 0 0.0188 -0.9998>}
 dyn_skirt_rear_04_r {parent=dyn_skirt_rear_03_r, loc=<15.2835 0 0>, rot=<0 0 0 1>}
 driver_skirt_rear_01_l {parent=pelvis, loc=<-0.255 -11.6668 -5.8725>, rot=<-0.0179 -0.9998 0 0>}
 dyn_skirt_rear_01_l {parent=driver_skirt_rear_01_l, loc=<0 0 0>, rot=<0 0 0 1>}
 dyn_skirt_rear_02_l {parent=dyn_skirt_rear_01_l, loc=<15.116 0 0>, rot=<-0.0079 -0.0006 0.0686 -0.9976>}
 dyn_skirt_rear_03_l {parent=dyn_skirt_rear_02_l, loc=<12.7812 0 0.0001>, rot=<0 0 0.0188 -0.9998>}
 dyn_skirt_rear_04_l {parent=dyn_skirt_rear_03_l, loc=<15.3262 0 0>, rot=<0 0 0 1>}
 AltSkin_corr_back {parent=pelvis, loc=<-13.0971 -10.9428 0>, rot=<-0.0127 -0.707 0.0127 -0.707>}
]
variables [
 BackwardMovement {type=double, default=0}
 Crouch {type=double, default=0}
 CrouchWalk {type=double, default=0}
 FE {type=bool, default=False}
 ForwardBackwardMove {type=double, default=0}
 ForwardMovement {type=double, default=0}
 Is parachute opened {type=bool, default=False}
 IsALT {type=bool, default=False}
 IsDBNO {type=bool, default=False}
 IsEmote {type=bool, default=False}
 IsSkyDive {type=bool, default=False}
 LeftMovement {type=double, default=0}
 LeftRightMove {type=double, default=0}
 RightMovement {type=double, default=0}
 Velocity character {type=double, default=0}
 isBackPack {type=bool, default=False}
 isFloating {type=bool, default=False}
]
