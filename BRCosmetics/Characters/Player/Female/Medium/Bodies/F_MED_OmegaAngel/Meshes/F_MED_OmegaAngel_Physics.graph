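# Physics asset for the F_MED_OmegaAngel female player body, setting up box-body chains for the front loincloth and rear skirt.
# Small driver bodies locked to the pelvis anchor limited-swing chains: a front piece branching left and right, and four-segment rear chains.
# Cross constraints tie left and right tips together; loincloth_03_r links to 01_r, skipping 02_r.
# Pelvis and calf capsules plus a floor-like root box provide body shapes; thighs have none.
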PhysicsAsset /BRCosmetics/Characters/Player/Female/Medium/Bodies/F_MED_OmegaAngel/Meshes/F_MED_OmegaAngel_Physics
bodies {
 calf_l [{
   shapeType=Capsule
   length=8.4106
   radius=7.1425
   shapeName=calf_l_capsule
   center=<-17.3137 -4.5033 -6.0166>
   rotation=<rot -42.7952 -0.0002 -180>
  }]
 calf_r [{shapeType=Capsule, length=8.4106, radius=7.1425, shapeName=calf_r_capsule, center=<17.3137 4.5033 6.0166>, rotation=<rot -42.7952 -0.0002 0>}]
 driver_skirt_rear_01_l [{shapeType=Sphere, radius=0.505, shapeName=driver_skirt_rear_01_l_sphere}]
 driver_skirt_rear_01_r [{shapeType=Sphere, radius=0.505, shapeName=driver_skirt_rear_01_r_sphere}]
 dyn_cloth_font_mid_01 [{
   shapeType=Box
   extentX=9.8549
   extentY=3.5696
   extentZ=13.9892
   shapeName=dyn_cloth_font_mid_01_box
   center=<5.0798 0.1099 1.6478>
   rotation=<rot 0 -0.5228 -88.1185>
  }]
 dyn_driver_cloth_front [{shapeType=Capsule, length=0.505, radius=0.505, shapeName=dyn_driver_cloth_front_capsule, rotation=<rot 0 -90 0>}]
 dyn_loincloth_01_l [{
   shapeType=Box
   extentX=9.0692
   extentY=4.807
   extentZ=20.1286
   shapeName=dyn_loincloth_01_l_box
   center=<8.9114 0.8419 0.1549>
   rotation=<rot 1.0651 87.2124 -87.2205>
  }]
 dyn_loincloth_01_r [{
   shapeType=Box
   extentX=9.0692
   extentY=5.92
   extentZ=20.1286
   shapeName=dyn_loincloth_01_l_box
   center=<-8.8223 -0.842 -0.7651>
   rotation=<rot 1.0651 87.2124 -89.8825>
  }]
 dyn_loincloth_02_l [{
   shapeType=Box
   extentX=14.1132
   extentY=5.7009
   extentZ=14.1857
   shapeName=dyn_loincloth_02_l_box
   center=<6.6595 3 1.0548>
   rotation=<rot 2.1374 0 90>
  }]
 dyn_loincloth_02_r [{
   shapeType=Box
   extentX=14.1132
   extentY=5.7009
   extentZ=14.1857
   shapeName=dyn_loincloth_02_r_box
   center=<-6.6595 -3 -1.0548>
   rotation=<rot 2.0539 0 -90>
  }]
 dyn_loincloth_03_l [{
   shapeType=Box
   extentX=11.4477
   extentY=4.5051
   extentZ=13.9419
   shapeName=dyn_loincloth_03_l_box
   center=<4.4886 4.7965 0.9101>
   rotation=<rot -0.3502 134.9148 -84.5737>
  }]
 dyn_loincloth_03_r [{
   shapeType=Box
   extentX=11.4477
   extentY=4.5051
   extentZ=13.9419
   shapeName=dyn_loincloth_03_r_box
   center=<-4.4886 -4.7965 -0.9101>
   rotation=<rot -0.3502 134.9148 95.4263>
  }]
 dyn_skirt_rear_01_l [{
   shapeType=Box
   extentX=3.1385
   extentY=13.0191
   extentZ=13.9439
   shapeName=dyn_skirt_rear_01_l_box
   center=<7.1737 0 0>
   rotation=<rot 0 -90 0>
  }]
 dyn_skirt_rear_01_r [{shapeType=Box, extentX=3.0493, extentY=13.197, extentZ=11.718, center=<7.2282 0 0>, rotation=<rot 0 -90 0>}]
 dyn_skirt_rear_02_l [{shapeType=Box, extentX=5.7993, extentY=10.98, extentZ=13.895, shapeName=dyn_skirt_rear_02_l_box, center=<5.5683 0 0>, rotation=<rot 0 90 0>}]
 dyn_skirt_rear_02_r [{
   shapeType=Box
   extentX=5.9899
   extentY=11.4812
   extentZ=11.6338
   shapeName=dyn_skirt_rear_02_r_box
   center=<5.5683 0 0>
   rotation=<rot 0 -90 0>
  }]
 dyn_skirt_rear_03_l [{shapeType=Box, extentX=3.7285, extentY=13.0283, extentZ=15.5953, shapeName=dyn_skirt_rear_03_l_box, center=<6 0 0>, rotation=<rot 0 90 -90>}]
 dyn_skirt_rear_03_r [{shapeType=Box, extentX=3.8485, extentY=14.7486, extentZ=16.0171, shapeName=dyn_skirt_rear_03_r_box, center=<6 0 0>, rotation=<rot 0 -90 -90>}]
 dyn_skirt_rear_04_l [{
   shapeType=Box
   extentX=4.6506
   extentY=13.7786
   extentZ=27.5207
   shapeName=dyn_skirt_rear_04_l_box
   center=<13 -2 0>
   rotation=<rot 0 -98.048 88.201>
  }]
 dyn_skirt_rear_04_r [{
   shapeType=Box
   extentX=3.6552
   extentY=13.7786
   extentZ=27.8446
   shapeName=dyn_skirt_rear_04_r_box
   center=<13 -2 0>
   rotation=<rot 0 -98.0483 -89.4147>
  }]
 pelvis [{shapeType=Capsule, length=14.0027, radius=8.3892, shapeName=pelvis_capsule, center=<-2.8436 -0.8926 -0.0172>}]
 root [{shapeType=Box, extentX=216.1274, extentY=166.3139, extentZ=10.8054, shapeName=root_box, center=<0 0 -5.7909>, rotation=<rot 0 -90 0>}]
 thigh_l []
 thigh_r []
}
constraints [
 {bone1Name=dyn_loincloth_02_l, bone2Name=dyn_loincloth_01_l, swing1LimitDegrees=10, swing2LimitDegrees=25, twistLimitDegrees=20}
 {bone1Name=dyn_loincloth_03_l, bone2Name=dyn_loincloth_02_l, swing1LimitDegrees=1, swing2LimitDegrees=20, twistLimitDegrees=20}
 {bone1Name=dyn_cloth_font_mid_01, bone2Name=dyn_driver_cloth_front, swing1LimitDegrees=10, swing2LimitDegrees=20, twistLimitDegrees=20}
 {bone1Name=dyn_loincloth_02_r, bone2Name=dyn_loincloth_01_r, swing1LimitDegrees=10, swing2LimitDegrees=25, twistLimitDegrees=20}
 {bone1Name=dyn_loincloth_03_r, bone2Name=dyn_loincloth_01_r, swing1LimitDegrees=1, swing2LimitDegrees=20, twistLimitDegrees=20}
 {bone1Name=dyn_loincloth_03_r, bone2Name=dyn_loincloth_03_l, swing1LimitDegrees=2, swing2LimitDegrees=2, twistLimitDegrees=20}
 {bone1Name=dyn_loincloth_01_l, bone2Name=dyn_cloth_font_mid_01, swing1LimitDegrees=10, swing2LimitDegrees=30, twistLimitDegrees=30}
 {bone1Name=dyn_loincloth_01_r, bone2Name=dyn_cloth_font_mid_01, swing1LimitDegrees=10, swing2LimitDegrees=30, twistLimitDegrees=30}
 {
  bone1Name=dyn_driver_cloth_front
  bone2Name=pelvis
  swing1LimitDegrees=45
  swing1Motion=Locked
  swing2LimitDegrees=45
  swing2Motion=Locked
  twistLimitDegrees=45
  twistMotion=Locked
 }
 {
  bone1Name=thigh_l
  bone2Name=pelvis
  swing1LimitDegrees=45
  swing1Motion=Free
  swing2LimitDegrees=45
  swing2Motion=Free
  twistLimitDegrees=45
  twistMotion=Free
 }
 {
  bone1Name=calf_l
  bone2Name=pelvis
  swing1LimitDegrees=45
  swing1Motion=Locked
  swing2LimitDegrees=45
  swing2Motion=Locked
  twistLimitDegrees=45
  twistMotion=Locked
 }
 {bone1Name=dyn_skirt_rear_01_l, bone2Name=driver_skirt_rear_01_l, swing1LimitDegrees=30, swing2LimitDegrees=10, twistLimitDegrees=1}
 {
  bone1Name=driver_skirt_rear_01_r
  bone2Name=pelvis
  swing1LimitDegrees=45
  swing1Motion=Locked
  swing2LimitDegrees=45
  swing2Motion=Locked
  twistLimitDegrees=45
  twistMotion=Locked
 }
 {
  bone1Name=driver_skirt_rear_01_l
  bone2Name=pelvis
  swing1LimitDegrees=45
  swing1Motion=Locked
  swing2LimitDegrees=45
  swing2Motion=Locked
  twistLimitDegrees=45
  twistMotion=Locked
 }
 {
  bone1Name=pelvis
  bone2Name=root
  swing1LimitDegrees=45
  swing1Motion=Locked
  swing2LimitDegrees=45
  swing2Motion=Locked
  twistLimitDegrees=45
  twistMotion=Locked
 }
 {bone1Name=dyn_skirt_rear_01_r, bone2Name=driver_skirt_rear_01_r, swing1LimitDegrees=30, swing2LimitDegrees=10, twistLimitDegrees=1}
 {bone1Name=dyn_skirt_rear_02_r, bone2Name=dyn_skirt_rear_01_r, swing1LimitDegrees=30, swing2LimitDegrees=10, twistLimitDegrees=10}
 {bone1Name=dyn_skirt_rear_03_r, bone2Name=dyn_skirt_rear_02_r, swing1LimitDegrees=20, swing2LimitDegrees=10, twistLimitDegrees=1}
 {bone1Name=dyn_skirt_rear_04_r, bone2Name=dyn_skirt_rear_03_r, swing1LimitDegrees=25, swing2LimitDegrees=10, twistLimitDegrees=1}
 {bone1Name=dyn_skirt_rear_04_l, bone2Name=dyn_skirt_rear_03_l, swing1LimitDegrees=25, swing2LimitDegrees=10, twistLimitDegrees=1}
 {bone1Name=dyn_skirt_rear_02_l, bone2Name=dyn_skirt_rear_01_l, swing1LimitDegrees=30, swing2LimitDegrees=10, twistLimitDegrees=10}
 {bone1Name=dyn_skirt_rear_03_l, bone2Name=dyn_skirt_rear_02_l, swing1LimitDegrees=20, swing2LimitDegrees=10, twistLimitDegrees=1}
 {bone1Name=dyn_skirt_rear_04_l, bone2Name=dyn_skirt_rear_04_r, swing1LimitDegrees=1, swing2LimitDegrees=1, twistLimitDegrees=20}
]
