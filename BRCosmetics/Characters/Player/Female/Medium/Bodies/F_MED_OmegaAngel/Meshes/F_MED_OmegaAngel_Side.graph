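# Physics asset for the side skirt of the OmegaAngel female medium body cosmetic, with capsules on skirt bones, pelvis, thighs and calves.
# Front-mid, side-mid and side chains on each side hang from Base bones through tight swing and twist limits.
# Right chains route through Driver bones on angularly free joints, while left chains link directly.
# Swing-locked cross-constraints tie neighboring lower segments to the side chain, keeping each half together.
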PhysicsAsset /BRCosmetics/Characters/Player/Female/Medium/Bodies/F_MED_OmegaAngel/Meshes/F_MED_OmegaAngel_Side
bodies {
 Base_skirt_frontMid_l [{shapeType=Capsule, length=0.505, radius=0.505, shapeName=Base_skirt_frontMid_l_capsule, rotation=<rot 0 -90 0>}]
 Base_skirt_frontMid_r [{shapeType=Capsule, length=0.505, radius=0.505, shapeName=Base_skirt_frontMid_r_capsule, rotation=<rot 0 -90 0>}]
 Driver_skirt_frontMid_02_r [{shapeType=Capsule, length=0.505, radius=0.505, shapeName=Driver_skirt_frontMid_02_r_capsule, rotation=<rot 0 -90 0>}]
 Driver_skirt_sideMid_02_r [{shapeType=Capsule, length=0.505, radius=0.505, shapeName=Driver_skirt_sideMid_02_r_capsule, rotation=<rot 0 -90 0>}]
 Driver_skirt_side_02_r1 [{shapeType=Capsule, length=0.505, radius=0.505, shapeName=Driver_skirt_side_02_r1_capsule, rotation=<rot 0 -90 0>}]
 calf_l [
  {shapeType=Capsule, length=11.8071, radius=7.0592, center=<-17.2673 -6.4527 -2.8673>, rotation=<rot -6.7043 -80.457 -145.5433>}
  {
   shapeType=Capsule
   length=28.0794
   radius=6.9181
   shapeName=calf_l_capsule
   center=<-33.5497 0 0.0658>
   rotation=<rot 89.055 0.0002 -179.9996>
  }
 ]
 calf_r [
  {shapeType=Capsule, length=8.076, radius=7.0592, center=<17.2673 6.4527 2.8673>, rotation=<rot -6.7043 -80.457 34.4567>}
  {shapeType=Capsule, length=28.0794, radius=6.9181, shapeName=calf_r_capsule, center=<33.5497 0 -0.0658>, rotation=<rot 89.055 0.0002 0.0004>}
 ]
 dyn_skirt_frontMid_01_l [{
   shapeType=Capsule
   length=17.4719
   radius=5.6772
   shapeName=dyn_skirt_frontMid_01_l_capsule
   center=<-5.1572 0 3>
   rotation=<rot 86.6886 -179.9998 -179.9997>
  }]
 dyn_skirt_frontMid_01_r [{shapeType=Capsule, length=1, radius=1}]
 dyn_skirt_frontMid_02_l [
  {shapeType=Capsule, length=13.7399, radius=5.6678, shapeName=dyn_skirt_frontMid_02_l_capsule, center=<-13 0 3>, rotation=<rot 90 0 0>}
  {
   shapeType=Capsule
   length=3.7319
   radius=5
   shapeName=dyn_skirt_frontMid_02_l_capsule
   center=<-21.0835 -3.5324 2.938>
   rotation=<rot 90 -0.028 0.0019>
  }
 ]
 dyn_skirt_frontMid_02_r [{shapeType=Capsule, length=13.0328, radius=5.6728, center=<5 1.1 -2.8>, rotation=<rot 86.0092 -179.9671 -179.9503>}]
 dyn_skirt_frontMid_03_r [
  {shapeType=Capsule, length=13, radius=5, center=<13 0 -3>, rotation=<rot 90 0 0>}
  {shapeType=Capsule, length=5, radius=5, shapeName=dyn_skirt_frontMid_03_r_capsule, center=<21 4 -3>, rotation=<rot -90 0 0>}
 ]
 dyn_skirt_sideMid_01_l [{
   shapeType=Capsule
   length=12.1833
   radius=4.7499
   shapeName=dyn_skirt_sideMid_01_l_capsule
   center=<-4.077 1.5 1.7>
   rotation=<rot 83.8386 -179.9998 -179.9999>
  }]
 dyn_skirt_sideMid_01_r [{shapeType=Capsule, length=1, radius=1}]
 dyn_skirt_sideMid_02_l [{shapeType=Capsule, length=13.696, radius=5.1134, shapeName=dyn_skirt_sideMid_02_l_capsule, center=<-13.0147 0 0>, rotation=<rot 90 0 0>}]
 dyn_skirt_sideMid_02_r [{shapeType=Capsule, length=10.2754, radius=3.9929, center=<4 0 -1.5>, rotation=<rot 84.0022 -179.9996 -179.9997>}]
 dyn_skirt_sideMid_03_l [{shapeType=Capsule, length=1, radius=1}]
 dyn_skirt_sideMid_03_l1 [{shapeType=Capsule, length=1, radius=1}]
 dyn_skirt_sideMid_03_r [{shapeType=Capsule, length=15.7853, radius=3.9647, center=<13.0147 0 0>, rotation=<rot 90 0 0>}]
 dyn_skirt_side_01_l [{shapeType=Capsule, length=1, radius=1}]
 dyn_skirt_side_01_r [{shapeType=Capsule, length=1, radius=1}]
 dyn_skirt_side_02_l [{shapeType=Capsule, length=13.75, radius=5, shapeName=dyn_skirt_side_02_l_capsule, center=<-4 0 2.5>, rotation=<rot 90 178.7341 -178.7342>}]
 dyn_skirt_side_02_r [{shapeType=Capsule, length=13.75, radius=5, center=<4 0 -2.5>, rotation=<rot 87.1192 -179.9998 -179.9997>}]
 dyn_skirt_side_03_l [{
   shapeType=Capsule
   length=14.0353
   radius=4.7147
   shapeName=dyn_skirt_side_03_l_capsule
   center=<-13.4522 0 2.7>
   rotation=<rot 90 -2.5305 -2.5304>
  }]
 dyn_skirt_side_03_r [{shapeType=Capsule, length=14.0353, radius=4.7147, center=<13.4522 0 -2.7>, rotation=<rot 90 -2.5305 -2.5304>}]
 pelvis [{shapeType=Capsule, length=0, radius=9.9063}]
 thigh_l [{shapeType=Capsule, length=38.4964, radius=7.5949, center=<-27.2847 0.568 -0.2116>, rotation=<rot -0.5472 -88.3726 85.4421>}]
 thigh_r [{shapeType=Capsule, length=38.4964, radius=7.5949, center=<27.2847 -0.568 0.2116>, rotation=<rot -0.5472 -88.3726 -94.5579>}]
}
constraints [
 {bone1Name=dyn_skirt_frontMid_01_r, bone2Name=Base_skirt_frontMid_r, swing1LimitDegrees=10, swing2LimitDegrees=45, twistLimitDegrees=1}
 {bone1Name=dyn_skirt_frontMid_02_r, bone2Name=Driver_skirt_frontMid_02_r, swing1LimitDegrees=22, swing2LimitDegrees=22, twistLimitDegrees=2}
 {bone1Name=dyn_skirt_frontMid_03_r, bone2Name=dyn_skirt_frontMid_02_r, swing1LimitDegrees=22, swing2LimitDegrees=20, twistLimitDegrees=2}
 {bone1Name=dyn_skirt_side_03_r, bone2Name=dyn_skirt_side_02_r, swing1LimitDegrees=25, swing2LimitDegrees=20, twistLimitDegrees=2}
 {bone1Name=dyn_skirt_sideMid_01_r, bone2Name=Base_skirt_frontMid_r, swing1LimitDegrees=10, swing2LimitDegrees=10, twistLimitDegrees=1}
 {bone1Name=dyn_skirt_sideMid_02_r, bone2Name=Driver_skirt_sideMid_02_r, swing1LimitDegrees=20, swing2LimitDegrees=15, twistLimitDegrees=2}
 {bone1Name=dyn_skirt_sideMid_03_r, bone2Name=dyn_skirt_sideMid_02_r, swing1LimitDegrees=22, swing2LimitDegrees=20, twistLimitDegrees=2}
 {bone1Name=dyn_skirt_side_01_r, bone2Name=Base_skirt_frontMid_r, swing1LimitDegrees=10, swing2LimitDegrees=45, twistLimitDegrees=1}
 {bone1Name=dyn_skirt_side_02_r, bone2Name=Driver_skirt_side_02_r1, swing1LimitDegrees=22, swing2LimitDegrees=15, twistLimitDegrees=2}
 {bone1Name=dyn_skirt_sideMid_03_l, bone2Name=Base_skirt_frontMid_l, swing1LimitDegrees=10, swing2LimitDegrees=45, twistLimitDegrees=1}
 {bone1Name=dyn_skirt_frontMid_01_l, bone2Name=dyn_skirt_sideMid_03_l, swing1LimitDegrees=20, swing2LimitDegrees=30, twistLimitDegrees=2}
 {bone1Name=dyn_skirt_frontMid_02_l, bone2Name=dyn_skirt_frontMid_01_l, swing1LimitDegrees=22, swing2LimitDegrees=35, twistLimitDegrees=2}
 {bone1Name=dyn_skirt_sideMid_03_l1, bone2Name=Base_skirt_frontMid_l, swing1LimitDegrees=10, swing2LimitDegrees=10, twistLimitDegrees=1}
 {bone1Name=dyn_skirt_sideMid_01_l, bone2Name=dyn_skirt_sideMid_03_l1, swing1LimitDegrees=25, swing2LimitDegrees=25, twistLimitDegrees=2}
 {bone1Name=dyn_skirt_sideMid_02_l, bone2Name=dyn_skirt_sideMid_01_l, swing1LimitDegrees=22, swing2LimitDegrees=25, twistLimitDegrees=2}
 {bone1Name=dyn_skirt_side_01_l, bone2Name=Base_skirt_frontMid_l, swing1LimitDegrees=10, swing2LimitDegrees=45, twistLimitDegrees=1}
 {bone1Name=dyn_skirt_side_02_l, bone2Name=dyn_skirt_side_01_l, swing1LimitDegrees=30, swing2LimitDegrees=25, twistLimitDegrees=2}
 {bone1Name=dyn_skirt_side_03_l, bone2Name=dyn_skirt_side_02_l, swing1LimitDegrees=30, swing2LimitDegrees=20, twistLimitDegrees=2}
 {
  bone1Name=dyn_skirt_sideMid_02_l
  bone2Name=dyn_skirt_side_03_l
  swing1LimitDegrees=45
  swing1Motion=Locked
  swing2LimitDegrees=45
  swing2Motion=Locked
  twistLimitDegrees=10
 }
 {
  bone1Name=dyn_skirt_frontMid_02_l
  bone2Name=dyn_skirt_side_03_l
  swing1LimitDegrees=45
  swing1Motion=Locked
  swing2LimitDegrees=45
  swing2Motion=Locked
  twistLimitDegrees=45
  twistMotion=Locked
 }
 {
  bone1Name=dyn_skirt_frontMid_03_r
  bone2Name=dyn_skirt_side_03_r
  swing1LimitDegrees=45
  swing1Motion=Locked
  swing2LimitDegrees=45
  swing2Motion=Locked
  twistLimitDegrees=45
  twistMotion=Locked
 }
 {
  bone1Name=dyn_skirt_sideMid_03_r
  bone2Name=dyn_skirt_side_03_r
  swing1LimitDegrees=45
  swing1Motion=Locked
  swing2LimitDegrees=45
  swing2Motion=Locked
  twistLimitDegrees=10
 }
 {
  bone1Name=Driver_skirt_side_02_r1
  bone2Name=dyn_skirt_side_01_r
  swing1LimitDegrees=45
  swing1Motion=Free
  swing2LimitDegrees=45
  swing2Motion=Free
  twistLimitDegrees=45
  twistMotion=Free
 }
 {
  bone1Name=Driver_skirt_sideMid_02_r
  bone2Name=dyn_skirt_sideMid_01_r
  swing1LimitDegrees=45
  swing1Motion=Free
  swing2LimitDegrees=45
  swing2Motion=Free
  twistLimitDegrees=45
  twistMotion=Free
 }
 {
  bone1Name=Driver_skirt_frontMid_02_r
  bone2Name=dyn_skirt_frontMid_01_r
  swing1LimitDegrees=45
  swing1Motion=Free
  swing2LimitDegrees=45
  swing2Motion=Free
  twistLimitDegrees=45
  twistMotion=Free
 }
]
