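# Control Rig for the F_MED_OmegaAngel face accessory that repositions the halo bone in certain body poses.
# A single ModifyTransforms applies a fixed rotation and translation offset to the halo, weighted procedurally from the current pose.
# The weight multiplies two opposing remaps of pelvis height above root, forming a lowered-pelvis band, by a remap of head Z rotation relative to spine_05.
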ControlRigBlueprint /BRCosmetics/Characters/Player/Female/Medium/Bodies/F_MED_OmegaAngel/Meshes/Parts/F_MED_OmegaAngel_FaceAcc_CtrlRig
graphs [
 RigVMFunctionLibrary {links [], nodes []}
 RigVMModel {nodes [
   RigUnit_BeginExecution = &RigUnit_BeginExecution()
   n1 = &RigUnit_GetRelativeTransformForItem(Child=(Type=Bone,Name="pelvis"), Parent=(Type=Bone,Name="root"), bChildInitial=False, bParentInitial=False)
   n2 = &RigVMFunction_MathDoubleRemap(SourceMaximum=67, SourceMinimum=90, TargetMaximum=1, TargetMinimum=0, Value=@n1.RelativeTransform.Translation.Z)
   n3 = &RigVMFunction_MathDoubleRemap(SourceMaximum=69, SourceMinimum=50, TargetMaximum=1, TargetMinimum=0, Value=@n1.RelativeTransform.Translation.Z)
   n4 = &RigVMFunction_MathQuaternionToEuler(
    RotationOrder=XYZ
    Value=&RigUnit_GetRelativeTransformForItem(Child=(Type=Bone,Name="head"), Parent=(Type=Bone,Name="spine_05"), bChildInitial=False, bParentInitial=False).RelativeTransform.Rotation
   )
   &RigUnit_ModifyTransforms(
    ItemToModify=((Item=(Type=Bone,Name="halo"),Transform=(Rotation=(X=-0.2201,Y=0,Z=0,W=0.9755),Translation=(X=0,Y=-0.6012,Z=0.1851),Scale3D=(X=1,Y=1,Z=1))))
    WeightMinimum=0
    Weight=&RigVMFunction_MathDoubleMul(
     A=&RigVMFunction_MathDoubleMul(B=@n2.Result, A=@n3.Result).Result
     B=&RigVMFunction_MathDoubleRemap(SourceMaximum=-44, SourceMinimum=-38, TargetMaximum=1, TargetMinimum=0, Value=@n4.Result.Z).Result
    ).Result
    ExecutePin=@RigUnit_BeginExecution.ExecutePin
   )
  ]}
]
hierarchy [
 root {loc=<0 0 0>, rot=<0 0 0 1>}
 pelvis {parent=root, loc=<0 -0.5301 98.0571>, rot=<0.0127 0.707 -0.0127 -0.707>}
 spine_01 {parent=pelvis, loc=<6.1208 -0.9328 0>, rot=<0 0 0.034 -0.9994>}
 spine_02 {parent=spine_01, loc=<6.5997 -0.2095 0>, rot=<0 0 0.0238 -0.9997>}
 spine_03 {parent=spine_02, loc=<6.4914 -0.1237 0>, rot=<0 0 0.0149 -0.9999>}
 spine_04 {parent=spine_03, loc=<6.9598 0.0794 0>, rot=<0 0 -0.0132 -0.9999>}
 spine_05 {parent=spine_04, loc=<6.9987 0 0>, rot=<0 0 -0.0015 -1>}
 neck_01 {parent=spine_05, loc=<11.3138 0.1529 0>, rot=<0 0 -0.2003 -0.9797>}
 neck_02 {parent=neck_01, loc=<4.5627 0 0>, rot=<0 0 0.0504 -0.9987>}
 head {parent=neck_02, loc=<5.3914 0 0>, rot=<0 0 0.115 -0.9934>}
 dyn_hair_05_l {parent=head, loc=<-1.3646 -2.5366 -6.8727>, rot=<0.0357 -0.0038 0.9993 0.0056>}
 dyn_hair_06_l {parent=dyn_hair_05_l, loc=<2.0986 0 0>, rot=<0 0 0 1>}
 dyn_hair_01_r {parent=head, loc=<1.2846 2.3497 7.3968>, rot=<-0.0687 -0.9945 0.0344 -0.0711>}
 dyn_hair_02_r {parent=dyn_hair_01_r, loc=<5.4276 0 0>, rot=<0 0 0 1>}
 dyn_hair_03_r {parent=dyn_hair_02_r, loc=<2.7991 -0.5729 -0.2741>, rot=<0 -0.0863 0 -0.9963>}
 dyn_hair_04_r {parent=dyn_hair_02_r, loc=<2.4966 0.6884 0.8466>, rot=<-0.2284 0.1666 -0.1601 -0.9457>}
 dyn_hair_05_r {parent=head, loc=<-0.3072 -2.4646 6.9508>, rot=<-0.0483 -0.9844 -0.1631 -0.0445>}
 dyn_hair_06_r {parent=dyn_hair_05_r, loc=<2.962 0 0>, rot=<0 0 0 1>}
 dyn_hair_01_l {parent=head, loc=<0.3285 2.2399 -7.6418>, rot=<0.0836 -0.0975 0.9909 0.0397>}
 dyn_hair_02_l {parent=dyn_hair_01_l, loc=<4.2744 0 0>, rot=<0 0 0 1>}
 dyn_hair_03_l {parent=dyn_hair_02_l, loc=<2.4115 0.3429 0>, rot=<0 0 0 1>}
 dyn_hair_04_l {parent=dyn_hair_02_l, loc=<2.3185 -1.345 0.6>, rot=<0.1403 0 0 -0.9901>}
 halo {parent=head, loc=<5.7378 0 0>, rot=<0.0029 -0.7071 -0.0029 -0.7071>}
 clavicle_r {parent=spine_05, loc=<5.4831 -0.1053 3.2946>, rot=<0.7289 0.0465 -0.6829 0.009>}
]
